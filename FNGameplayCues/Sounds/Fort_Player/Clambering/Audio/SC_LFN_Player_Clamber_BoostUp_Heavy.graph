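# Sound cue for the player's heavy clamber boost-up, layering a pitched-up random boost-up sound with a random heavy variant.
# Islocal and IsEnemy branches route it to first-person, friendly or enemy foley sound classes, with unset parameters falling through to enemy.
# Friendly playback is much quieter than enemy.
# Remote paths share one mix with separate whoosh and clamber attenuation assets per layer, while local playback has its own mix.
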SoundCue /FNGameplayCues/Sounds/Fort_Player/Clambering/Audio/SC_LFN_Player_Clamber_BoostUp_Heavy
root=Branch_1
nodes [
 {
  id=Attenuation_0
  children [Modulator_4]
  properties {attenuationSettings=/FNGameplayCues/Sounds/Attenuation/Fort_Player_Footsteps_Clamber_Whoosh_Att}
 }
 {id=Attenuation_2, children [Random_6], properties {attenuationSettings=/FNGameplayCues/Sounds/Attenuation/Fort_Player_Footsteps_Clamber_Att}}
 {id=Attenuation_8, children [Modulator_9]}
 {id=Branch_0, children [SoundClass_3,SoundClass_2,SoundClass_3], properties {boolParameterName=IsEnemy}}
 {id=Branch_1, children [SoundClass_0,Branch_0,Branch_0], properties {boolParameterName=Islocal}}
 {id=Mixer_2, children [Attenuation_0, null, Attenuation_2], properties {inputVolume [1,1.8,1]}}
 {id=Mixer_3, children [Modulator_5], properties {inputVolume [1]}}
 {id=Mixer_4, children [Modulator_5], properties {inputVolume [0.2]}}
 {id=Mixer_5, children [Attenuation_8], properties {inputVolume [1]}}
 {id=Mixer_6, children [Modulator_8,Random_9], properties {inputVolume [1,1]}}
 {id=Modulator_4, children [Random_4], properties {pitchMax=1.3, pitchMin=1.2}}
 {id=Modulator_5, children [Mixer_2], properties {pitchMax=1.2, volumeMax=1, volumeMin=1}}
 {id=Modulator_8, children [Random_7], properties {pitchMax=1.3, pitchMin=1.2}}
 {id=Modulator_9, children [Mixer_6], properties {pitchMax=1.2, volumeMax=1, volumeMin=1}}
 {
  id=Random_4
  children [WavePlayer_20, WavePlayer_21, WavePlayer_22, WavePlayer_23, WavePlayer_24]
  properties {bShouldExcludeFromBranchCulling=true, weights [1,1,1,1,1]}
 }
 {
  id=Random_6
  children [WavePlayer_31, WavePlayer_32, WavePlayer_33, WavePlayer_34]
  properties {bShouldExcludeFromBranchCulling=true, weights [1,1,1,1]}
 }
 {
  id=Random_7
  children [WavePlayer_35, WavePlayer_36, WavePlayer_37, WavePlayer_38, WavePlayer_39]
  properties {bShouldExcludeFromBranchCulling=true, weights [1,1,1,1,1]}
 }
 {
  id=Random_9
  children [WavePlayer_46, WavePlayer_47, WavePlayer_48, WavePlayer_49]
  properties {bShouldExcludeFromBranchCulling=true, weights [1,1,1,1]}
 }
 {id=SoundClass_0, children [Mixer_5], properties {soundClassOverride=&Fort_Player_Foley_1P_Athena}}
 {
  id=SoundClass_2
  children [Mixer_4]
  properties {soundClassOverride=/Game/Sounds/FortSoundClassesAndModes/Fort_Player_Foley_3P_Friendly_Athena}
 }
 {id=SoundClass_3, children [Mixer_3], properties {soundClassOverride=&Fort_Player_Foley_3P_Enemy_Athena}}
 {id=WavePlayer_20, properties {soundWaveAssetPtr=&/Player_Clamber_BoostUp_05}}
 {id=WavePlayer_21, properties {soundWaveAssetPtr=&/Player_Clamber_BoostUp_01}}
 {id=WavePlayer_22, properties {soundWaveAssetPtr=&/Player_Clamber_BoostUp_02}}
 {id=WavePlayer_23, properties {soundWaveAssetPtr=&/Player_Clamber_BoostUp_03}}
 {id=WavePlayer_24, properties {soundWaveAssetPtr=&/Player_Clamber_BoostUp_04}}
 {id=WavePlayer_31, properties {soundWaveAssetPtr=&/Player_Clamber_BoostUp_Heavy_01}}
 {id=WavePlayer_32, properties {soundWaveAssetPtr=&/Player_Clamber_BoostUp_Heavy_02}}
 {id=WavePlayer_33, properties {soundWaveAssetPtr=&/Player_Clamber_BoostUp_Heavy_03}}
 {id=WavePlayer_34, properties {soundWaveAssetPtr=&/Player_Clamber_BoostUp_Heavy_04}}
 {id=WavePlayer_35, properties {soundWaveAssetPtr=&Player_Clamber_BoostUp_05}}
 {id=WavePlayer_36, properties {soundWaveAssetPtr=&Player_Clamber_BoostUp_01}}
 {id=WavePlayer_37, properties {soundWaveAssetPtr=&Player_Clamber_BoostUp_02}}
 {id=WavePlayer_38, properties {soundWaveAssetPtr=&Player_Clamber_BoostUp_03}}
 {id=WavePlayer_39, properties {soundWaveAssetPtr=&Player_Clamber_BoostUp_04}}
 {id=WavePlayer_46, properties {soundWaveAssetPtr=&Player_Clamber_BoostUp_Heavy_01}}
 {id=WavePlayer_47, properties {soundWaveAssetPtr=&Player_Clamber_BoostUp_Heavy_02}}
 {id=WavePlayer_48, properties {soundWaveAssetPtr=&Player_Clamber_BoostUp_Heavy_03}}
 {id=WavePlayer_49, properties {soundWaveAssetPtr=&Player_Clamber_BoostUp_Heavy_04}}
]
properties {maxDistance=2097152, volumeMultiplier=0.5}
