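# Sound cue for the default boost-up foley when a player clambers.
# Branches on Islocal, then IsEnemy, route to first-person, friendly or enemy sound classes, with unset IsEnemy treated as enemy.
# Local playback mixes randomized default boost-up waves with delayed clothing.
# Enemy and friendly paths share an attenuated mix adding a pitched-up, delayed boost-up layer, with friendlies far quieter through a lower mixer volume.
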SoundCue /FNGameplayCues/Sounds/Fort_Player/Clambering/Audio/SC_Player_Clamber_BoostUp_Default
root=Branch_4
nodes [
 {
  id=Attenuation_0
  children [Modulator_0]
  properties {attenuationSettings=/FNGameplayCues/Sounds/Attenuation/Fort_Player_Footsteps_Clamber_Whoosh_Att}
 }
 {
  id=Attenuation_1
  children [Random_1]
  properties {attenuationSettings=&Fort_Player_Footsteps_Clamber_Whoosh_Att}
 }
 {id=Attenuation_2, children [Random_4], properties {attenuationSettings=/FNGameplayCues/Sounds/Attenuation/Fort_Player_Footsteps_Clamber_Att}}
 {id=Attenuation_8, children [Modulator_6]}
 {id=Branch_4, children [SoundClass_4,Branch_5,Branch_5], properties {boolParameterName=Islocal}}
 {id=Branch_5, children [SoundClass_3,SoundClass_2,SoundClass_3], properties {boolParameterName=IsEnemy}}
 {id=Delay_0, children [Attenuation_1], properties {delayMax=0.2, delayMin=0.2}}
 {id=Delay_1, children [Random_6], properties {delayMax=0.05, delayMin=0.05}}
 {id=Delay_2, children [Random_0], properties {delayMax=0.15, delayMin=0.15}}
 {id=Mixer_0, children [Attenuation_0, Delay_0, Attenuation_2], properties {inputVolume [1.5,1.25,0.35]}}
 {id=Mixer_2, children [Modulator_3], properties {inputVolume [1]}}
 {id=Mixer_3, children [Modulator_3], properties {inputVolume [0.2]}}
 {id=Mixer_4, children [Attenuation_8], properties {inputVolume [1]}}
 {id=Mixer_6, children [Delay_1,Modulator_8], properties {inputVolume [1,1.5]}}
 {id=Modulator_0, children [Delay_2], properties {pitchMax=1.3, pitchMin=1.2}}
 {id=Modulator_3, children [Mixer_0], properties {pitchMax=1.2, volumeMax=1, volumeMin=1}}
 {id=Modulator_6, children [Mixer_6], properties {pitchMax=1.2, volumeMax=1, volumeMin=1}}
 {id=Modulator_8, children [Random_7], properties {pitchMax=1.2, volumeMax=1, volumeMin=0.8}}
 {
  id=Random_0
  children [WavePlayer_0,WavePlayer_1,WavePlayer_2,WavePlayer_3,WavePlayer_4]
  properties {bShouldExcludeFromBranchCulling=true, weights [1,1,1,1,1]}
 }
 {
  id=Random_1
  children [WavePlayer_5, WavePlayer_6, WavePlayer_7, WavePlayer_8, WavePlayer_9, WavePlayer_10]
  properties {bShouldExcludeFromBranchCulling=true, weights [1,1,1,1,1,1]}
 }
 {
  id=Random_4
  children [WavePlayer_20, WavePlayer_21, WavePlayer_22, WavePlayer_23]
  properties {bShouldExcludeFromBranchCulling=true, weights [1,1,1,1]}
 }
 {
  id=Random_6
  children [WavePlayer_29, WavePlayer_30, WavePlayer_31, WavePlayer_32, WavePlayer_33, WavePlayer_34]
  properties {bShouldExcludeFromBranchCulling=true, weights [1,1,1,1,1,1]}
 }
 {
  id=Random_7
  children [WavePlayer_35, WavePlayer_36, WavePlayer_37, WavePlayer_38]
  properties {bShouldExcludeFromBranchCulling=true, weights [1,1,1,1]}
 }
 {
  id=SoundClass_2
  children [Mixer_3]
  properties {soundClassOverride=/Game/Sounds/FortSoundClassesAndModes/Fort_Player_Foley_3P_Friendly_Athena}
 }
 {id=SoundClass_3, children [Mixer_2], properties {soundClassOverride=&Fort_Player_Foley_3P_Enemy_Athena}}
 {id=SoundClass_4, children [Mixer_4], properties {soundClassOverride=&Fort_Player_Foley_1P_Athena}}
 {id=WavePlayer_0, properties {soundWaveAssetPtr=&/Player_Clamber_BoostUp_05}}
 {id=WavePlayer_1, properties {soundWaveAssetPtr=&/Player_Clamber_BoostUp_01}}
 {id=WavePlayer_10, properties {soundWaveAssetPtr=&/Clambering_Foley_Clothing_06}}
 {id=WavePlayer_2, properties {soundWaveAssetPtr=&/Player_Clamber_BoostUp_02}}
 {id=WavePlayer_20, properties {soundWaveAssetPtr=&/Player_Clamber_BoostUp_Default_01}}
 {id=WavePlayer_21, properties {soundWaveAssetPtr=&/Player_Clamber_BoostUp_Default_02}}
 {id=WavePlayer_22, properties {soundWaveAssetPtr=&/Player_Clamber_BoostUp_Default_03}}
 {id=WavePlayer_23, properties {soundWaveAssetPtr=&/Player_Clamber_BoostUp_Default_04}}
 {id=WavePlayer_29, properties {soundWaveAssetPtr=&/Clambering_Foley_Clothing_01}}
 {id=WavePlayer_3, properties {soundWaveAssetPtr=&/Player_Clamber_BoostUp_03}}
 {id=WavePlayer_30, properties {soundWaveAssetPtr=&/Clambering_Foley_Clothing_02}}
 {id=WavePlayer_31, properties {soundWaveAssetPtr=&/Clambering_Foley_Clothing_03}}
 {id=WavePlayer_32, properties {soundWaveAssetPtr=&/Clambering_Foley_Clothing_04}}
 {id=WavePlayer_33, properties {soundWaveAssetPtr=&/Clambering_Foley_Clothing_05}}
 {id=WavePlayer_34, properties {soundWaveAssetPtr=&Clambering_Foley_Clothing_06}}
 {id=WavePlayer_35, properties {soundWaveAssetPtr=&Player_Clamber_BoostUp_Default_01}}
 {id=WavePlayer_36, properties {soundWaveAssetPtr=&Player_Clamber_BoostUp_Default_02}}
 {id=WavePlayer_37, properties {soundWaveAssetPtr=&Player_Clamber_BoostUp_Default_03}}
 {id=WavePlayer_38, properties {soundWaveAssetPtr=&Player_Clamber_BoostUp_Default_04}}
 {id=WavePlayer_4, properties {soundWaveAssetPtr=&/Player_Clamber_BoostUp_04}}
 {id=WavePlayer_5, properties {soundWaveAssetPtr=&Clambering_Foley_Clothing_01}}
 {id=WavePlayer_6, properties {soundWaveAssetPtr=&Clambering_Foley_Clothing_02}}
 {id=WavePlayer_7, properties {soundWaveAssetPtr=&Clambering_Foley_Clothing_03}}
 {id=WavePlayer_8, properties {soundWaveAssetPtr=&Clambering_Foley_Clothing_04}}
 {id=WavePlayer_9, properties {soundWaveAssetPtr=&Clambering_Foley_Clothing_05}}
]
properties {maxDistance=2097152, volumeMultiplier=1}
